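# Physics asset for the Ancient Gladiator back bling pack, providing capsule bodies for the pack, a hanging axe, a quiver and three arrows.
# The axe and quiver are constrained to the pack, and each arrow is constrained to the quiver, forming a short chain.
# Every joint locks twist and one swing axis, leaving a single narrow swing so the parts sway slightly, tighter for the arrows.
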PhysicsAsset /BRCosmetics/Accessories/FORT_Backpacks/Backpack_F_MED_Ancient_Gladiator/Meshes/F_MED_Ancient_Gladiator_Pack_Physics
bodies {
 dyn_arrow_jnt_1 [{shapeType=Capsule, length=5, radius=1, center=<0.8242 0 0>, rotation=<rot -90 0 0>}]
 dyn_arrow_jnt_2 [{shapeType=Capsule, length=5, radius=1, center=<0.8242 0 0>, rotation=<rot -90 0 0>}]
 dyn_arrow_jnt_3 [{shapeType=Capsule, length=5, radius=1, center=<0.8242 0 0>, rotation=<rot -90 0 0>}]
 dyn_axe_jnt [{shapeType=Capsule, length=30, radius=2.5, center=<0.6184 -0.0016 -0.0315>, rotation=<rot 0 -90.1467 -87.0844>}]
 dyn_quiver_jnt [{shapeType=Capsule, length=17.7355, radius=4.6087, center=<-0.0154 0.0001 0.333>, rotation=<rot 0 89.6645 -87.3564>}]
 pack [{shapeType=Capsule, length=3, radius=7, center=<0.0001 -5.0596 -7.7574>, rotation=<rot 0 -90 0>}]
}
constraints [
 {
  bone1Name=dyn_axe_jnt
  bone2Name=pack
  swing1LimitDegrees=45
  swing1Motion=Locked
  swing2LimitDegrees=7
  twistLimitDegrees=45
  twistMotion=Locked
 }
 {
  bone1Name=dyn_quiver_jnt
  bone2Name=pack
  swing1LimitDegrees=45
  swing1Motion=Locked
  swing2LimitDegrees=7
  twistLimitDegrees=45
  twistMotion=Locked
 }
 {
  bone1Name=dyn_arrow_jnt_1
  bone2Name=dyn_quiver_jnt
  swing1LimitDegrees=45
  swing1Motion=Locked
  swing2LimitDegrees=4
  twistLimitDegrees=45
  twistMotion=Locked
 }
 {
  bone1Name=dyn_arrow_jnt_2
  bone2Name=dyn_quiver_jnt
  swing1LimitDegrees=45
  swing1Motion=Locked
  swing2LimitDegrees=4
  twistLimitDegrees=45
  twistMotion=Locked
 }
 {
  bone1Name=dyn_arrow_jnt_3
  bone2Name=dyn_quiver_jnt
  swing1LimitDegrees=45
  swing1Motion=Locked
  swing2LimitDegrees=4
  twistLimitDegrees=45
  twistMotion=Locked
 }
]
